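; BGA_Thumper is a thumper prop that summons a small-creature mound encounter around itself.
; On the authority, it waits after BeginPlay, enables the point provider on its child-actor Living World volume, then starts the Lager encounter through the GameState a tick later.
; On remotes, a timeline moves the oil drum mesh vertically, playing an impact sound and restarting each time it finishes to create a looping thump.
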(event EventBeginPlay
  (Networking|SwitchHasAuthority
    (:Authority
      (Utilities|FlowControl|Delay 6.0)
      (bind _asfort_athena_living_world_volume (Utilities|Casting|CastToFortAthenaLivingWorldVolume (Variables|/Script/Engine.ChildActorComponent|GetChildActor (Variables|Default|GetLivingWorldVolume))))
      (Class|FortAthenaLivingWorldPointProviderInterface|EnablePointProvider_DEPRECATED _asfort_athena_living_world_volume)
      (Utilities|FlowControl|DelayUntilNextTick)
      (Default|StartActivityEncounter (Game|GetGameState) _asfort_athena_living_world_volume self "/Lager/Lager/Encounters/S37/LagerEncounter_Mound_SmallCreature.LagerEncounter_Mound_SmallCreature"))
    (:Remote
      (bind (_direction _floatvalue) (Timeline|ThumperTimeline Play)
        (:Update
          (Transformation|SetRelativeLocation (Variables|Default|GetS_Constr_OilDrum_01) (Math|Vector|MakeVector 0.0 0.0 (* _floatvalue 8e+01)) true false))
        (:Finished
          (Audio|PlaySoundatLocation "/Game/Sounds/Fort_GamePlay_Sounds/BRSword/AB_Impact_Cue.AB_Impact_Cue" (Transformation|GetActorLocation))
          (Components|Timeline|PlayfromStart (Variables|BGA_Thumper|GetThumperTimeline)))))))

(fn ConstructionScript)
